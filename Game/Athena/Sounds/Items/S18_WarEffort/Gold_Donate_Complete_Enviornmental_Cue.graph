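# Sound cue for the environmental sound of a completed gold donation in the Season 18 War Effort.
# A single wave player feeds the output directly, routed to the non-critical in-game UI sound class with a slight volume boost.
# Inline attenuation overrides the referenced prop-search asset, using logarithmic distance falloff and reverb send without spatialization.
# A concurrency override allows one instance at a time and blocks new plays until it ends.
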SoundCue /Game/Athena/Sounds/Items/S18_WarEffort/Gold_Donate_Complete_Enviornmental_Cue
root=WavePlayer_1
nodes [{id=WavePlayer_1, properties {soundWaveAssetPtr=&/Gold_Donate_Complete_Env}}]
properties {
 attenuationSettings=/Game/Sounds/Attenuation/Fort_Prop_Search_Att
 bOverrideAttenuation=true
 bOverrideConcurrency=true
 maxDistance=5600
 soundClassObject=&Fort_UI_InGame_NonCritical
 volumeMultiplier=1.1
 attenuationOverrides {
  bSpatialize=false
  distanceAlgorithm=Logarithmic
  reverbDistanceMax=3000
  reverbDistanceMin=0
  reverbWetLevelMax=0.5
  reverbWetLevelMin=0.1
  attenuationShapeExtents {x=2000}
 }
 concurrencyOverrides {maxCount=1, resolutionRule=PreventNew}
}
