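# Sound cue for the Season 18 War Effort gold donation interaction.
# A random node picks one of four Gold_Donate_Interact wave variations with equal weight.
# Playback runs at reduced volume through the non-critical in-game UI sound class, spatialized with the prop search attenuation preset.
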SoundCue /Game/Athena/Sounds/Items/S18_WarEffort/Gold_Donate_Interact_Cue
root=Random_0
nodes [
 {id=Random_0, children [WavePlayer_1,WavePlayer_2,WavePlayer_3,WavePlayer_4], properties {weights [1,1,1,1]}}
 {id=WavePlayer_1, properties {soundWaveAssetPtr=&/Gold_Donate_Interact_01}}
 {id=WavePlayer_2, properties {soundWaveAssetPtr=&/Gold_Donate_Interact_02}}
 {id=WavePlayer_3, properties {soundWaveAssetPtr=&/Gold_Donate_Interact_03}}
 {id=WavePlayer_4, properties {soundWaveAssetPtr=&/Gold_Donate_Interact_04}}
]
properties {
 attenuationSettings=/Game/Sounds/Attenuation/Fort_Prop_Search_Att
 maxDistance=1300
 soundClassObject=&Fort_UI_InGame_NonCritical
 volumeMultiplier=0.65
}
